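# Material function that builds the asphalt surface for the Apollo landscape as material attributes.
# It projects asphalt diffuse, SMR and normal textures along world Z, darkening base color with a separately scaled world-aligned crack mask.
# A shading path switch gives Forward and Mobile a cheaper variant with flat normal, constant roughness and specular from diffuse green.
# Crack height and an intensity-adjusted crack normal are exposed as separate outputs.
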MaterialFunction /Clyde/Athena/Apollo/Environments/Landscape/Materials/MaterialFunctions/MF_ApolloAsphalt
graph {nodes [
  n1 = TextureObjectParameter() {group=Asphalt, parameterName=Asphalt Diffuse, texture=/Clyde/Athena/Apollo/Environments/Landscape/Textures/Roads/T_AsphaltDetail1_D}
  n2 = TextureObjectParameter() {
   group=Asphalt
   parameterName=Asphalt_SMR
   samplerType=SAMPLERTYPE_Masks
   texture=/Clyde/Athena/Apollo/Environments/Landscape/Textures/Roads/T_AsphaltDetail1_S
  }
  n3 = TextureObjectParameter() {
   group=Asphalt
   parameterName=Asphalt Normal
   samplerType=SAMPLERTYPE_Normal
   texture=/Clyde/Athena/Apollo/Environments/Landscape/Textures/Roads/T_AsphaltDetail1_N
  }
  n4 = MaterialFunctionCall!(
   Diffuse Texture Object=@n1
   Spec Texture Object=@n2
   TextureSize=FunctionInput(Preview=Constant() {r=350}) {bUsePreviewValueAsDefault=true, inputName=TilingSize, inputType=FunctionInput_Scalar}
   --------------------------
   Use seperate settings for Z projection
   Use seperate texture for Z projection
   Texture Size Z projection
   Z Plane Diffuse Texture Object
   Z Plane Spec Texture Object
   "--------------------------      "
   ProjectionTransitionContrast
   Use Texture Alpha Heightmap to Refine the Transistion=StaticBool()
   HeightLerpTransitionContrast
   PickTheDiffuseHeightChannel
   TheDiffuseHeightChannel
   "--------------------------       "
   NormalTextureObject=@n3
   World Space Normals=StaticBool()
   "--------------------------  "
   Export Float 4 Diffuse=StaticBool()
   Export Float 4 Spec
   "--------------------------     "
   WorldPosition=WorldPosition() {worldPositionShaderOffset=WPT_ExcludeAllShaderOffsets}.XYZ
   World Space Normal
  ) {materialFunction=/Engine/Functions/Engine_MaterialFunctions02/Texturing/WorldAlignedTextures_Complex}
  n5 = ScalarParameter() {defaultValue=0.5905, parameterName=AsphaltVariation}
  n6 = ScalarParameter() {defaultValue=2000, parameterName=SizeCracks}
  n7 = MaterialFunctionCall(
   TextureObject=TextureObject() {samplerType=SAMPLERTYPE_Masks, texture=/Clyde/Athena/Apollo/Environments/Landscape/Textures/Roads/Road_Cracks_M}
   TextureSize=@n6
  ) {materialFunction=&WorldAlignedTexture}
  n8 = ComponentMask(None=@n7.Z Texture) {r=true}
  n9 = Multiply(A=ComponentMask(None=@n4.Z Diffuse Texture) {b=true, g=true, r=true}, B=OneMinus(None=Multiply(A=@n5, B=@n8)))
  n10 = SetMaterialAttributes!(
   MaterialAttributes
   BaseColor=@n9
   Specular=ComponentMask(None=@n4.Z Spec) {r=true}
   Roughness=ComponentMask(None=@n4.Z Spec) {b=true}
   Normal=@n4.Normal Z Texture
  ) {attributeSetTypes [
    #1
    #2
    #3
    #4
   ]}
  n11 = SetMaterialAttributes(
   BaseColor=@n9
   Specular=ComponentMask(None=@n4.Z Diffuse Texture) {g=true}
   Roughness=Constant() {r=0.7}
   Normal=Constant3Vector() {constant=<rgba 0 0 1 0>}
  ) {attributeSetTypes [
    #1
    #2
    #3
    #4
   ]}
  FunctionOutput(None=ShadingPathSwitch(Default=@n10, Deferred=@n10, Forward=@n11, Mobile=@n11))
  n12 = TextureObject() {samplerType=SAMPLERTYPE_Normal, texture=/Clyde/Athena/Apollo/Environments/Landscape/Textures/Roads/Road_Cracks_N}
  FunctionOutput(None=MaterialFunctionCall(
    Normal=MaterialFunctionCall(TextureObject=@n12, TextureSize=@n6) {materialFunction=&WorldAlignedNormal}.Z Texture
    Flatness=ScalarParameter() {defaultValue=-4, parameterName=AspahltCrackNormal}
   ) {materialFunction=&FlattenNormal}.Result) {outputName=CrackNormal, sortPriority=8}
  FunctionOutput(None=ComponentMask(None=@n7.Z Texture) {g=true}) {outputName=CrackHeight, sortPriority=4}
 ]}
